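# Default subtitle widget Blueprint, derived from SubtitleWidget, that lays out on-screen dialog subtitles, closed captions and descriptions.
# An Overlay root holds three centered Borders with translucent grey backgrounds: dialog at the bottom, captions at the top, descriptions in the middle.
# Each Border wraps an inherited TextBlock, with caption and description text tinted grey.
# It only supplies layout and styling; its event graph contains just empty event stubs.
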WidgetBlueprint /SubtitlesAndClosedCaptions/Widgets/DefaultSubtitleWidget
graphs [EventGraph {nodes [
   Event PreConstruct()
   Event Construct()
   Event Tick()
  ]}]
variables []
widgets {
 info {inheritedWidgetCount=6, parentClass=/Script/SubtitlesAndClosedCaptions.SubtitleWidget, rootWidgetClass=&Overlay, widgetCount=7}
 widgets [
  Overlay_33 {widgetClassPath=&Overlay, widget=Overlay_33}
  DialogBorder {
   widgetClassPath=&Border
   bInherited=true
   parent=Overlay_33
   widget=DialogBorder
   properties {
    horizontalAlignment=HAlign_Center
    verticalAlignment=VAlign_Bottom
    background {
     bIsDynamicallyLoaded=false
     drawAs=Image
     imageType=NoImage
     mirroring=NoMirror
     resourceName=None
     resourceObject=null
     tiling=NoTile
     imageSize=<32 32>
     margin=<ltrb 0 0 0 0>
     outlineSettings {
      bUseBrushTransparency=false
      roundingType=HalfHeightRadius
      width=0
      color {colorUseRule=UseColor_Specified, specifiedColor=<rgba 0 0 0 0>}
      cornerRadii=<0 0 0 0>
     }
     tintColor {colorUseRule=UseColor_Specified, specifiedColor=<rgba 0.1865 0.1865 0.1865 0.8>}
     uVRegion {bIsValid=false, max=<0 0>, min=<0 0>}
    }
    padding=<ltrb 0 0 0 0>
   }
   slot {class=OverlaySlot, properties {horizontalAlignment=HAlign_Center, verticalAlignment=VAlign_Bottom}}
  }
  DialogSubtitleBlock {
   widgetClassPath=&TextBlock
   bInherited=true
   parent=DialogBorder
   widget=DialogSubtitleBlock
   properties {justification=Center, minDesiredWidth=1280, text=Text Block, margin=<ltrb 16 16 16 16>}
   slot {class=BorderSlot, properties {horizontalAlignment=HAlign_Center, verticalAlignment=VAlign_Bottom, padding=<ltrb 0 0 0 0>}}
  }
  CaptionBorder {
   widgetClassPath=&Border
   bInherited=true
   parent=Overlay_33
   widget=CaptionBorder
   properties {
    horizontalAlignment=HAlign_Center
    verticalAlignment=VAlign_Top
    background {
     bIsDynamicallyLoaded=false
     drawAs=Image
     imageType=NoImage
     mirroring=NoMirror
     resourceName=None
     resourceObject=null
     tiling=NoTile
     imageSize=<32 32>
     margin=<ltrb 0 0 0 0>
     outlineSettings {
      bUseBrushTransparency=false
      roundingType=HalfHeightRadius
      width=0
      color {colorUseRule=UseColor_Specified, specifiedColor=<rgba 0 0 0 0>}
      cornerRadii=<0 0 0 0>
     }
     tintColor {colorUseRule=UseColor_Specified, specifiedColor=<rgba 0.1865 0.1865 0.1865 0.8>}
     uVRegion {bIsValid=false, max=<0 0>, min=<0 0>}
    }
    padding=<ltrb 0 0 0 0>
   }
   slot {class=OverlaySlot, properties {horizontalAlignment=HAlign_Center}}
  }
  CaptionSubtitleBlock {
   widgetClassPath=&TextBlock
   bInherited=true
   parent=CaptionBorder
   widget=CaptionSubtitleBlock
   properties {
    justification=Center
    minDesiredWidth=1280
    text=Caption Text Block
    colorAndOpacity {colorUseRule=UseColor_Specified, specifiedColor=<rgba 0.5 0.5 0.5 1>}
    margin=<ltrb 16 16 16 16>
   }
   slot {class=BorderSlot, properties {horizontalAlignment=HAlign_Center, verticalAlignment=VAlign_Top, padding=<ltrb 0 0 0 0>}}
  }
  DescriptionBorder {
   widgetClassPath=&Border
   bInherited=true
   parent=Overlay_33
   widget=DescriptionBorder
   properties {
    horizontalAlignment=HAlign_Center
    verticalAlignment=VAlign_Center
    background {
     bIsDynamicallyLoaded=false
     drawAs=Image
     imageType=NoImage
     mirroring=NoMirror
     resourceName=None
     resourceObject=null
     tiling=NoTile
     imageSize=<32 32>
     margin=<ltrb 0 0 0 0>
     outlineSettings {
      bUseBrushTransparency=false
      roundingType=HalfHeightRadius
      width=0
      color {colorUseRule=UseColor_Specified, specifiedColor=<rgba 0 0 0 0>}
      cornerRadii=<0 0 0 0>
     }
     tintColor {colorUseRule=UseColor_Specified, specifiedColor=<rgba 0.1865 0.1865 0.1865 0.8>}
     uVRegion {bIsValid=false, max=<0 0>, min=<0 0>}
    }
    padding=<ltrb 0 0 0 0>
   }
   slot {class=OverlaySlot, properties {horizontalAlignment=HAlign_Center, verticalAlignment=VAlign_Center}}
  }
  DescriptionSubtitleBlock {
   widgetClassPath=&TextBlock
   bInherited=true
   parent=DescriptionBorder
   widget=DescriptionSubtitleBlock
   properties {text=Description Text Block, colorAndOpacity {colorUseRule=UseColor_Specified, specifiedColor=<rgba 0.5 0.5 0.5 1>}}
   slot {class=BorderSlot, properties {horizontalAlignment=HAlign_Center, verticalAlignment=VAlign_Center, padding=<ltrb 0 0 0 0>}}
  }
 ]
}
